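; Local gameplay behavior that plays a looping montage on its activating actor, for the WildEstate ClimbSpill non-participant boss bot.
; On execute it resolves the boss pawn through the AI bot controller, shows its prop and plays the first Montage Array entry.
; A timer set to the montage length replays that entry, toggling a MontageSwitch flag that nothing here reads.
; Abort hides the prop and stops the montage but does not clear the timer.
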(event EventBeginPlay)

(event Collision|EventActorBeginOverlap
  (params (OtherActor :type Actor%20Object%20Reference)))

(event EventTick
  (params (DeltaSeconds :type Float%20%28single-precision%29)))

(event Custom|PlayNewMontage
  (CallFunction|LGBPlayMontage (Utilities|Array|Get(acopy) (Variables|Default|GetMontage%20Array))))

(event LocalGameplayBehavior|EventOnExecute
  (params (Activator :type Actor%20Object%20Reference))
  (bind _output_get (Variables|Default|SetActivatingActor Activator))
  (bind _asfort_athena_aibot_controller (Utilities|Casting|CastToFortAthenaAIBotController _output_get))
  (bind _asbp_player_pawn_non_participant_boss_wild_estate_climb_spill (Utilities|Casting|CastToBP_PlayerPawn_NonParticipant_Boss_WildEstate_ClimbSpill (Variables|/Script/FortniteAI.FortAthenaAIBotController|GetPlayerBotPawn _asfort_athena_aibot_controller)))
  (Variables|Default|SetClimbSpillPawn _asbp_player_pawn_non_participant_boss_wild_estate_climb_spill)
  (Class|BPPlayerPawnNonParticipantBossWildEstateClimbSpill|SetPropVisibiilty _asbp_player_pawn_non_participant_boss_wild_estate_climb_spill true)
  (CallFunction|LGBPlayMontage (Utilities|Array|Get(acopy) (Variables|Default|GetMontage%20Array))))

(event Custom|LGB_PlayMontage
  (params (Anim%20Montage :type Anim%20Montage%20Object%20Reference))
  (bind _output_get (Variables|Default|SetPlayingMontage Anim%20Montage))
  (bind _returnvalue (Animation|Utilities|PlayAnimMontage (Variables|Default|GetActivatingActor) _output_get))
  (Utilities|Time|SetTimerbyEvent (EventDispatchers|CreateEvent :!event Custom|OnMontageEnd) _returnvalue))

(event Custom|OnMontageEnd
  (Variables|Default|SetMontageSwitch (not (Variables|Default|GetMontageSwitch)))
  (CallFunction|PlayNewMontage))

(event LocalGameplayBehavior|EventOnAbort
  (bind _climbspillpawn (Variables|Default|GetClimbSpillPawn))
  (Utilities|IsValid _climbspillpawn
    (:Is%20Valid
      (Class|BPPlayerPawnNonParticipantBossWildEstateClimbSpill|SetPropVisibiilty _climbspillpawn false)
      (Animation|Utilities|StopMontageifCurrent (Variables|Default|GetActivatingActor) (Variables|Default|GetPlayingMontage)))))

(fn ConstructionScript)
